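# Unlit additive effect material in the water fluid simulation that generates a field of randomly placed rain drop dots.
# UVs tile by Drop Density, and each cell gets one SphereMask dot, with its radius scaled by density to keep UV size constant.
# A custom HLSL hash of the cell index plus stepped, looping time jitters each dot's centre, so drops jump to new positions rather than moving smoothly.
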Material /FortWater/FluidSimulation/Materials/Effects/M_Rain_01
graph {
 properties {
  bAllowNegativeEmissiveColor=true
  bUsedWithNanite=true
  blendMode=BLEND_Additive
  pixelDepthOffsetMode=PDOM_Legacy
  shadingModel=MSM_Unlit
  usedShadingModels=Unlit
 }
 nodes [
  n1 = Multiply(A=TextureCoordinate(), B=ScalarParameter() {defaultValue=3, group=Drops, parameterName=Drop Density}) {constB=8}
  n2 = Constant() {r=0.5}
  n3 = Divide(
   A=MaterialFunctionCall(Speed=ScalarParameter() {defaultValue=0.1, parameterName=Speed}) {materialFunction=&TimeWithSpeedVariable}.Frac Time
   B=ScalarParameter() {defaultValue=0.005, parameterName=Divisor}
  )
  MaterialExpressionAdd_2 = Add(A=Ceil(None=@n1), B=Multiply(A=Ceil(None=@n3), B=ScalarParameter() {defaultValue=1, parameterName=Mult}))
  n4 = Custom!(pos=@MaterialExpressionAdd_2) {
   code="float2 noise = dot(pos.xy * pos.xy, float2(12.581057, 11.129465));\r\nnoise.y = dot(pos.xy * pos.xy/5, float2(46.538326, 11.057759));\r\n\r\nreturn frac(noise);"
   description=Pseudo Random
   outputType=CMOT_Float2
   inputs [{inputName=pos, input {expression=&self:MaterialExpressionAdd_2}}]
  }
  n5 = Multiply(A=ConstantBiasScale(None=@n4) {bias=-0.5, scale=2}, B=ScalarParameter() {defaultValue=0.475, parameterName=Random})
  n6 = ScalarParameter() {defaultValue=3, group=Drops, parameterName=Drop Density}
  n7 = Divide(A=ScalarParameter() {defaultValue=1, group=Drops, parameterName=Radius}) {constB=256}
  Multiply(
   A=SphereMask(
    A=Frac(None=@n1)
    B=Add(A=@n2, B=@n5)
    Radius=Multiply(A=@n6, B=@n7)
    Hardness=ScalarParameter() {defaultValue=0.5, group=Drops, parameterName=Hardness}
   )
   B=ScalarParameter() {defaultValue=1, parameterName=F}
  )
 ]
}
