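# M_Fluid_Display_01 is a subsurface-shaded water surface material that displays a fluid simulation through a NormalAndHeight texture.
# It blends depth-attenuated Gerstner waves with the texture's height and normals, using an exponential depth falloff against MPC_Water's TargetWaveDepth.
# Inverted texture alpha serves as foam for color, scattering and roughness terms.
# View-dependent Fresnel and bent-vector dot products fake internal refraction and top-down scattering between two tinted colors.
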Material /FortWater/FluidSimulation/Materials/M_Fluid_Display_01
graph {
 properties {
  allowTranslucentCustomDepthWrites=true
  bUsedWithNanite=true
  opacityMaskClipValue=0.001
  pixelDepthOffsetMode=PDOM_Legacy
  refractionMethod=RM_PixelNormalOffset
  shadingModel=MSM_Subsurface
  translucencyLightingMode=TLM_Surface
  usedShadingModels=Subsurface
 }
 nodes [
  n1 = MaterialFunctionCall!(Z Offset Only, Position, WaterBodyIndex, Normals use VS Interpolation, All Normals VS) {materialFunction=/FortWater/Materials/Functions/Fort_ComputeGerstnerWaves}
  MaterialExpressionDivide_5 = Divide(
   A=ScalarParameter() {defaultValue=1000, parameterName=Water Depth}
   B=Divide(A=CollectionParameter() {collection=/Water/Materials/MPC/MPC_Water, parameterName=TargetWaveDepth})
  )
  n2 = Custom!(x=@MaterialExpressionDivide_5) {
   code="return 1-exp(-x);"
   description=exp
   outputType=CMOT_Float1
   inputs [{inputName=x, input {expression=&self:MaterialExpressionDivide_5}}]
  }
  n3 = Multiply(A=@n2, B=ScalarParameter() {defaultValue=1, parameterName=Full Waves})
  n4 = TextureSampleParameter2D() {parameterName=NormalAndHeight, texture=&Black}
  n5 = VectorParameter() {parameterName=H, defaultValue=<rgba 0 0 1 0>}
  Add(A=Multiply(A=@n1.WPO, B=@n3), B=Multiply(A=Multiply(A=@n4.B, B=@n5.RGB)))
  n6 = OneMinus(None=TextureSampleParameter2D() {parameterName=NormalAndHeight, texture=&Black}.A)
  n7 = VectorParameter() {parameterName=Foam SSS, defaultValue=<rgba 0.5 0.7 1 0>}
  n8 = Multiply(A=@n6, B=Multiply(A=@n7.RGB, B=@n7.RGB))
  Add(
   A=VectorParameter() {parameterName=Base, defaultValue=<rgba 0 0.003 0.005 0>}.RGB
   B=Multiply(A=ScalarParameter() {parameterName=Foam Basecolor}, B=@n8)
  )
  n9 = Normalize(VectorInput=VectorParameter() {parameterName=testn, defaultValue=<rgba 0 0 -1 0>}.RGB)
  n10 = PixelNormalWS()
  n11 = OneMinus(None=Fresnel(ExponentIn=ScalarParameter() {defaultValue=1, parameterName=expn}))
  n12 = LinearInterpolate(
   A=CameraVectorWS()
   B=DeriveNormalZ(InXY=ComponentMask(None=@n10) {g=true, r=true})
   Alpha=LinearInterpolate(
    A=ScalarParameter() {defaultValue=1, parameterName=Refract Min}
    B=ScalarParameter() {defaultValue=0.5, parameterName=Refract Max}
    Alpha=@n11
   )
  )
  n13 = Power(Base=ConstantBiasScale(None=DotProduct(A=@n9, B=@n12)), Exp=ScalarParameter() {defaultValue=1, parameterName=test power})
  n14 = VectorParameter() {parameterName=Internal Refraction Color, defaultValue=<rgba 0 1.2 1 0>}
  n15 = VectorParameter() {parameterName=Internal Refraction Color 2, defaultValue=<rgba 0 0.5 0.4 2>}
  n16 = ConstantBiasScale(None=DotProduct(A=VectorParameter() {parameterName=gljk}.RGB, B=CameraVectorWS()))
  n17 = ScalarParameter() {defaultValue=2, parameterName=CamVec Power}
  n18 = Multiply(
   A=LinearInterpolate(A=Multiply(A=@n14.RGB, B=@n14.RGB), B=Multiply(A=@n15.RGB, B=@n15.RGB), Alpha=Power(Base=@n16, Exp=@n17))
   B=ScalarParameter() {defaultValue=6, parameterName=scatterboost}
  )
  n19 = Power(Base=@n11, Exp=ScalarParameter() {defaultValue=2, parameterName=p2})
  n20 = Fresnel(ExponentIn=ScalarParameter() {defaultValue=0.75, parameterName=TOP DOWN EXP})
  Add(A=@n8, B=Multiply(A=Multiply(A=@n13, B=@n18), B=Multiply(A=@n19, B=@n20)) {desc=Top down mask})
  Fresnel(BaseReflectFractionIn=ScalarParameter() {parameterName=BaseOpacity})
  n21 = ComponentMask(None=@n4.RGB) {g=true, r=true}
  MaterialFunctionCall(
   BaseNormal=Constant3Vector() {constant=<rgba 0 0 1 0>}
   AdditionalNormal=MaterialFunctionCall(Normal=DeriveNormalZ(InXY=@n21), Flatness=ScalarParameter() {parameterName=Normal Flatness}) {materialFunction=&FlattenNormal}.Result
  ) {materialFunction=&BlendAngleCorrectedNormals}
  MaterialFunctionCall(Normal=@n1.Normal, Flatness=OneMinus(None=@n3)) {materialFunction=&FlattenNormal}
  n22 = ScalarParameter() {defaultValue=0.2, parameterName=Foam Rough Width}
  Multiply(A=Saturate(None=Divide(A=@n6, B=@n22)), B=ScalarParameter() {defaultValue=0.2, parameterName=Foam R})
  ScalarParameter() {defaultValue=1.4, parameterName=Refration}
  ScalarParameter() {defaultValue=0.4, parameterName=spec}
  ScalarParameter() {parameterName=Roughness}
 ]
}
