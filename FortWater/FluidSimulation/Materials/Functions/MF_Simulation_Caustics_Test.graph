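# Test material function for underwater caustics, combining a tiling caustics texture with caustics derived from a fluid simulation's NormalAndHeight texture.
# A panned wave normal map distorts both layers' world XY coordinates, and simulation normals further warp the tiling layer.
# Simulation bands come from the difference between derived normal Z at two offset samples, thinned by AlphaOffset.
# The simulation lookup uses a hardcoded zero-centred extent, and color add and multiplier finish the result.
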MaterialFunction /FortWater/FluidSimulation/Materials/Functions/MF_Simulation_Caustics_Test
graph {
 properties {bExposeToLibrary=true, description=Caustics for Underwater}
 nodes [
  n1 = FunctionInput(Preview=WorldPosition().XYZ) {bUsePreviewValueAsDefault=true, inputName=Position}
  n2 = FunctionInput() {bUsePreviewValueAsDefault=true, inputName=Distortion Scale, inputType=FunctionInput_Scalar, sortPriority=1, previewValue=<750 0 0 1>}
  n3 = Divide(A=ComponentMask(None=@n1) {g=true, r=true}, B=@n2)
  n4 = FunctionInput() {
   bUsePreviewValueAsDefault=true
   inputName=Distortion Panning Speed
   inputType=FunctionInput_Scalar
   sortPriority=2
   previewValue=<0.05 0 0 1>
  }
  n5 = MaterialFunctionCall(Speed=@n4) {materialFunction=&TimeWithSpeedVariable}
  n6 = VectorParameter() {parameterName=pan2, defaultValue=<rgba 1 0.5 0 0>}
  n7 = ComponentMask(None=TextureSample(UVs=Panner(Coordinate=@n3, Time=@n5.Time, Speed=@n6.RGB)) {samplerType=SAMPLERTYPE_Normal, texture=/FortWater/Textures/T_Water_TilingNormal_Waves_02}.RGB) {g=true, r=true}
  n8 = FunctionInput() {
   bUsePreviewValueAsDefault=true
   inputName=Distortion on Tiling Caustics
   inputType=FunctionInput_Scalar
   sortPriority=5
   previewValue=<0.1 0 0 1>
  }
  n9 = Multiply(A=@n7, B=@n8)
  n10 = FunctionInput() {bUsePreviewValueAsDefault=true, inputName=Caustics Scale, inputType=FunctionInput_Scalar, sortPriority=3, previewValue=<500 0 0 1>}
  n11 = Divide(A=ComponentMask(None=@n1) {g=true, r=true}, B=@n10)
  n12 = FunctionInput() {
   bUsePreviewValueAsDefault=true
   inputName=Simulation Caustics Distort Texture Caustics
   inputType=FunctionInput_Scalar
   sortPriority=7
   previewValue=<0.1 0 0 1>
  }
  n13 = FunctionInput() {
   bUsePreviewValueAsDefault=true
   inputName=Distortion on Simulation Caustics
   inputType=FunctionInput_Scalar
   sortPriority=6
   previewValue=<16 0 0 1>
  }
  n14 = Add(A=Multiply(A=@n7, B=@n13), B=ComponentMask(None=@n1) {g=true, r=true})
  n15 = ComponentMask(None=Constant()) {g=true, r=true}
  n16 = ComponentMask(None=Constant() {r=1024}) {a=true}
  n17 = ConstantBiasScale(None=Divide(A=Subtract(A=@n14, B=@n15), B=Multiply(A=@n16) {constB=0.5}))
  n18 = ComponentMask(None=TextureSampleParameter2D(UVs=Clamp(None=@n17)) {parameterName=NormalAndHeight, texture=&Black}.RGB) {g=true, r=true}
  n19 = FunctionInput() {bUsePreviewValueAsDefault=true, inputName=Caustics Panning Speed, inputType=FunctionInput_Scalar, sortPriority=4, previewValue=<0.01 0 0 1>}
  n20 = Panner(
   Coordinate=Add(A=Add(A=@n9, B=@n11), B=Multiply(A=@n12, B=@n18))
   Time=MaterialFunctionCall(Speed=@n19) {materialFunction=&TimeWithSpeedVariable}.Time
   Speed=VectorParameter() {parameterName=pan1, defaultValue=<rgba 1 0.5 0 0>}.RGB
  )
  n21 = FunctionInput() {
   bUsePreviewValueAsDefault=true
   inputName=Simulation Caustics Brightness
   inputType=FunctionInput_Scalar
   sortPriority=9
   previewValue=<2 0 0 1>
  }
  n22 = OneMinus(None=DeriveNormalZ(InXY=@n18))
  n23 = ComponentMask(None=VectorParameter() {parameterName=UV Offset, defaultValue=<rgba 0.01 0.01 0 0>}.RGB) {g=true, r=true}
  n24 = TextureSampleParameter2D(UVs=Clamp(None=Add(A=@n17, B=@n23))) {parameterName=NormalAndHeight, texture=&Black}
  n25 = OneMinus(None=DeriveNormalZ(InXY=ComponentMask(None=@n24.RGB) {g=true, r=true}))
  n26 = Abs(None=Subtract(A=ComponentMask(None=@n22) {b=true}, B=ComponentMask(None=@n25) {b=true}))
  n27 = FunctionInput() {
   bUsePreviewValueAsDefault=true
   description=Used to make the sim caustic bands thinner. May cause aliasing.
   inputName=Sim Caustics Alpha Offset
   inputType=FunctionInput_Scalar
   sortPriority=11
   previewValue=<0.05 2 2 1>
  }
  n28 = Add(
   A=Multiply(
    A=TextureSample(UVs=@n20) {texture=/FortWater/Textures/T_Caustics_3layer}.RGB
    B=ScalarParameter() {defaultValue=1, parameterName=Base Caustics}
   )
   B=Multiply(
    A=@n21
    B=MaterialFunctionCall!(Alpha=@n26, Offset=@n27) {materialFunction=/Engine/Functions/Engine_MaterialFunctions01/Opacity/AlphaOffset}.Result
   )
  )
  n29 = FunctionInput() {bUsePreviewValueAsDefault=true, inputName=Color Add, sortPriority=9, previewValue=<0.5 0.5 0.5 1>}
  n30 = FunctionInput() {bUsePreviewValueAsDefault=true, inputName=Color Multiplier, sortPriority=10, previewValue=<2 2 2 1>}
  FunctionOutput(None=Multiply(A=Add(A=@n28, B=@n29), B=@n30))
  Multiply()
  ScalarParameter() {defaultValue=0.15, parameterName=alphaoffset}
  ScalarParameter() {defaultValue=500, parameterName=Caustics Scale}
  ScalarParameter() {defaultValue=0.01, parameterName=Caustic Speed}
  ScalarParameter() {defaultValue=16, parameterName=Distortion}
  ScalarParameter() {defaultValue=0.1, parameterName=Dist}
 ]
}
