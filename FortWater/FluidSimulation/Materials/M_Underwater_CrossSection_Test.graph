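# Test translucent unlit material for an underwater cross-section, shading the scene seen behind the surface.
# Scene color gets simulation caustics at the world position behind translucency, masked by normal Z, then spectrum tint and fog via depth-based Beer's law.
# A second output masks pixels below a water height offset by a texture sample, currently a black placeholder.
# Forward and mobile paths fall back to constants.
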Material /FortWater/FluidSimulation/Materials/M_Underwater_CrossSection_Test
graph {
 properties {bUsedWithNanite=true, blendMode=BLEND_Translucent, pixelDepthOffsetMode=PDOM_Legacy, shadingModel=MSM_Unlit, usedShadingModels=Unlit}
 nodes [
  ComponentMask(None=SceneTexture() {desc=Post Process Uses This, sceneTextureId=PPI_PostProcessInput0}.Color) {b=true, g=true, r=true}
  n1 = SceneColor!(UVs)
  n2 = Multiply(
   A=@n1.RGB
   B=MaterialFunctionCall!(
    Position=MaterialFunctionCall() {materialFunction=&WorldPositionBehindTranslucency}.Result
    Distortion Scale=ScalarParameter() {defaultValue=750, parameterName=Distortion Scale}
    Distortion Panning Speed=ScalarParameter() {defaultValue=0.05, parameterName=Distortion Speed}
    Caustics Scale=ScalarParameter() {defaultValue=500, parameterName=Caustics Texture Scale}
    Caustics Panning Speed=ScalarParameter() {defaultValue=0.01, parameterName=Caustic Speed}
    Distortion on Tiling Caustics=ScalarParameter() {defaultValue=0.1, parameterName=Distortion on Caustic Tex}
    Distortion on Simulation Caustics=ScalarParameter() {defaultValue=16, parameterName=Texture-Sim Distortion}
    Simulation Caustics Distort Texture Caustics=ScalarParameter() {defaultValue=0.1, parameterName=Sim Caustic Dist Texture Caustics}
    Simulation Caustics Brightness=ScalarParameter() {defaultValue=2, parameterName=Sim Caustics Brightness}
    Color Add=ScalarParameter() {defaultValue=0.2, parameterName=Caustic Fill}
    Color Multiplier=ScalarParameter() {defaultValue=2, parameterName=Caustic Boost}
    Sim Caustics Alpha Offset=ScalarParameter() {defaultValue=0.1, parameterName=alphaoffset}
   ) {materialFunction=/FortWater/FluidSimulation/Materials/Functions/MF_Simulation_Caustics_Test}.Result
  )
  n3 = Clamp(None=ComponentMask(None=SceneTexture() {sceneTextureId=PPI_WorldNormal}.Color) {b=true}) {desc=Mask for surfaces pointing down}
  n4 = VectorParameter() {parameterName=Spectrum Color, defaultValue=<rgba 0 0.6 1 0>}
  n5 = SceneDepth()
  n6 = MaterialFunctionCall!(Thickness=@n5.R, Depth Scale=ScalarParameter() {defaultValue=0.002, parameterName=Spectrum Density}) {materialFunction=/Engine/Functions/Engine_MaterialFunctions01/Volumetrics/BeersLaw}
  n7 = PixelDepth()
  n8 = LinearInterpolate(
   A=VectorParameter() {parameterName=Fog Color, defaultValue=<rgba 0.05 0.1 0.3 0>}.RGB
   B=Multiply(
    A=LinearInterpolate(A=@n1.RGB, B=@n2, Alpha=@n3) {constA=1, desc=Prevent Caustics on Underside of meshes}
    B=LinearInterpolate(A=@n4.RGB, Alpha=@n6.Result) {constA=1}
   )
   Alpha=MaterialFunctionCall(Thickness=Subtract(A=@n5.R, B=@n7.R), Depth Scale=ScalarParameter() {defaultValue=0.0004, parameterName=G depth}) {materialFunction=&BeersLaw}.Result
  )
  n9 = Constant3Vector()
  ShadingPathSwitch(Default=@n8, Deferred=@n8, Forward=@n9, Mobile=@n9)
  n10 = WorldPosition()
  n11 = Add(A=ScalarParameter() {defaultValue=-0.6, parameterName=Z Bias}, B=ComponentMask(None=@n10.XYZ) {b=true})
  n12 = Constant()
  n13 = Constant() {r=1024}
  n14 = Divide(
   A=Subtract(A=ComponentMask(None=@n10.XYZ) {g=true, r=true}, B=ComponentMask(None=@n12) {g=true, r=true})
   B=Multiply(A=ComponentMask(None=@n13) {a=true}) {constB=0.5}
  )
  n15 = Multiply(
   A=ScalarParameter() {defaultValue=-5, parameterName=Fluid Height}
   B=TextureSample(UVs=ConstantBiasScale(None=@n14)) {texture=&Black}.B
  )
  n16 = Clamp(None=Subtract(A=Constant() {r=64}, B=Add(A=@n11, B=@n15)))
  n17 = Constant() {r=1}
  ShadingPathSwitch(Default=@n16, Deferred=@n16, Forward=@n17, Mobile=@n17)
 ]
}
